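# Niagara idle effect for the right one-handed Neon Cat Tech pickaxe: a recurring cyan-blue sheen with flares.
# SheenMesh_Loop burst-spawns short-lived particles every five seconds after a randomized first delay, drawing through one mesh and three sprite renderers.
# Lifetime curves animate material parameters, color and size, plus separate flare attributes that sweep along Z, so each particle carries several layered elements.
# SheenMesh_Loop001 is a disabled, faster-looping variant with random mesh index.
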NiagaraSystem /BRCosmetics/Weapons/FORT_Melee/Pickaxe_Neon_Cat_Tech_Female_1H/FX/NS_Neon_Cat_Tech_Idle_1h_R
dependencies {
 usedDataInterfaces [&NiagaraDataInterfaceCurve, &NiagaraDataInterfaceColorCurve, &NiagaraDataInterfaceVectorCurve]
 usedDynamicInputs [
  &Multiply_Vector2
  &RandomRangeVector2D
  &Vector2DFromFloat
 ]
 usedModules [
  &self:SystemSpawnScript.SystemScriptSource.SystemScriptGraph.NiagaraNodeAssignment_2.SetVariables_#1
  &SystemState
 ]
}
emitters {
 SheenMesh_Loop {
  emitterUpdateScript [
   &EmitterState(
    Delay First Loop Only=true
    Life Cycle Mode=::NewEnumerator1
    Loop Delay=&RandomRangeFloat(Minimum=0, Maximum=1, Evaluation Type=::NewEnumerator1)
    Loop Duration=&RandomRangeFloat(Minimum=5, Maximum=5)
    Recalculate Duration Each Loop=true
    UseLoopDelay=true
   )
   &SpawnBurst_Instantaneous(Spawn Count=1, Use Spawn Probability=false)
   &SpawnBurst_Instantaneous(Spawn Count=1, Spawn Time=0.1, Use Spawn Probability=false) {module=SpawnBurst_Instantaneous001}
   &SpawnRate(SpawnRate=20, Use Spawn Probability=false) {enabled=false}
  ]
  particleSpawnScript [
   &Color(Color=&MakeLinearColorFromVectorAndFloat(
     "Vector (RGB)"=<1 0 0>
     "Float (Alpha)"=&ClampFloat(Float=&RandomRangeFloat(Minimum=0.4, Maximum=0.6, Evaluation Type=::NewEnumerator1))
    )) {enabled=false}
   /SetVariables(
    Particles.ColorFlare:LinearColor=&MultiplyLinearColor(Linear Color=$System.GlintColor, Scale Factor=<20 20 20 20>)
    Particles.PositionFlare:NiagaraPosition=$Particles.Position
   ) {isSetParameters=true, module=SetVariables_#2}
   /Niagara/Modules/Spawn/Initialization/InitializeParticle(
    Color=<rgba 0.0348 14.4774 20 1>
    Lifetime=&RandomRangeFloat(Minimum=1.5, Maximum=1.5)
    Module.Write Color=true
    Module.Write Lifetime=true
    Module.Write Mass=true
    Module.Write Material Random=true
    Module.Write Position=false
    Module.Write Scale=true
    Module.Write SpriteRotation=false
    Module.Write SpriteSize=true
    Position=<0 -1 -2>
    Sprite Size=<1 2>
   )
   /NiagaraFN/Sprite/FN_SpriteAlignmentAndFacing(SpriteAlignment:Vector3f=<0 0 1>, SpriteFacing:Vector3f=<1 0 0>)
  ]
  particleUpdateScript [
   &ParticleState()
   &ScaleSpriteSize(Scale Factor=&MakeVector2D(
     X=&FloatFromCurve(
      FloatCurve=DataInterface() {propertyValues {Curve {keys [
          {arriveTangent=7.953, interpMode=RCIM_Cubic, leaveTangent=7.953, tangentMode=RCTM_User, time=-0.0008, value=0.5}
          {arriveTangent=0.0662, interpMode=RCIM_Cubic, leaveTangent=0.0662, time=0.1, value=1.5834}
          {arriveTangent=-4.779, interpMode=RCIM_Cubic, leaveTangent=-4.779, time=0.2045, value=0.5136}
          {arriveTangent=-0.1355, interpMode=RCIM_Cubic, leaveTangent=-0.1355, tangentMode=RCTM_User, time=0.4, value=0.1497}
         ]}}}
      Scale Curve=10
     )
     Y=&FloatFromCurve(
      FloatCurve=DataInterface() {propertyValues {Curve {keys [
          {arriveTangent=7.953, interpMode=RCIM_Cubic, leaveTangent=7.953, tangentMode=RCTM_User, time=-0.0008, value=0.5}
          {arriveTangent=0.8674, interpMode=RCIM_Cubic, leaveTangent=0.8674, time=0.15, value=1.8998}
          {arriveTangent=-6.9225, interpMode=RCIM_Cubic, leaveTangent=-6.9225, time=0.2096, value=0.6825}
          {arriveTangent=-0.4678, interpMode=RCIM_Cubic, leaveTangent=-0.4678, tangentMode=RCTM_User, time=0.4, value=0.1692}
         ]}}}
      Scale Curve=10
     )
    ))
   /SetVariables(Particles.SpriteSizeFlare2:Vector2f=&MakeVector2D(
     X=&FloatFromCurve(
      FloatCurve=DataInterface() {propertyValues {Curve {keys [
          {arriveTangent=7.953, interpMode=RCIM_Cubic, leaveTangent=7.953, tangentMode=RCTM_User, time=-0.0008, value=0.5}
          {interpMode=RCIM_Cubic, leaveTangent=-0.7143, time=0.225, value=1.8998}
          {arriveTangent=-0.1355, interpMode=RCIM_Cubic, leaveTangent=-0.1355, tangentMode=RCTM_User, time=0.6, value=0.1497}
         ]}}}
      Scale Curve=10
     )
     Y=&FloatFromCurve(
      FloatCurve=DataInterface() {propertyValues {Curve {keys [
          {arriveTangent=7.953, interpMode=RCIM_Cubic, leaveTangent=7.953, tangentMode=RCTM_User, time=-0.0008, value=0.5}
          {interpMode=RCIM_Cubic, leaveTangent=-0.6804, time=0.225, value=1.8998}
          {arriveTangent=-0.4678, interpMode=RCIM_Cubic, leaveTangent=-0.4678, tangentMode=RCTM_User, time=0.6, value=0.1692}
         ]}}}
      Scale Curve=15
     )
    )) {isSetParameters=true, module=SetVariables_#3}
   &Color(Color=&MakeLinearColorFromVectorAndFloat("Vector (RGB)"=<0.4208 0.6361 1>, "Float (Alpha)"=&ClampFloat(Float=0.8))) {enabled=false, module=Color001}
   &DynamicMaterialParameters(
    Index 0 Param 1=&FloatFromCurve(FloatCurve=DataInterface() {propertyValues {Curve {keys [{arriveTangent=2.513, interpMode=RCIM_Cubic, time=0, value=0.2}, <at 0.3 1>]}}})
    Index 0 Param 2=0.5
    Index 0 Param 3=&FloatFromCurve(FloatCurve=DataInterface() {propertyValues {Curve {keys [{interpMode=RCIM_Cubic, tangentMode=RCTM_User, time=0.2, value=0}, <at 0.4 2 RCIM_Cubic>, <at 1 0 RCIM_Cubic>]}}})
    Index 0 Param 4=&FloatFromCurve(FloatCurve=DataInterface() {propertyValues {Curve {keys [
         {interpMode=RCIM_Cubic, tangentMode=RCTM_User, time=0, value=0}
         {interpMode=RCIM_Cubic, tangentMode=RCTM_User, time=0.8, value=0.4}
         {tangentMode=RCTM_User, time=1, value=0}
        ]}}})
    Module.Param0WriteEnabled=true
    Module.Param1WriteEnabled=true
    Module.Param2WriteEnabled=true
    Module.Param3WriteEnabled=true
   )
   &CameraOffset(Camera Offset Amount=1)
   /SetVariables(Particles.CameraOffsetFlare=2) {isSetParameters=true, module=SetVariables_#4}
   /SetVariables(
    Particles.ColorFlare:LinearColor=&ColorFromCurve(ColorCurve=DataInterface() {propertyValues {
       AlphaCurve {keys [<at 0.2801 1 RCIM_Cubic>,<at 0.5537 0 RCIM_Cubic>]}
       BlueCurve {keys [<at 0 1 RCIM_Cubic>,<at 0.1 1 RCIM_Cubic>]}
       GreenCurve {keys [<at 0 1 RCIM_Cubic>,<at 0.1 0.7239 RCIM_Cubic>]}
       RedCurve {keys [<at 0 1 RCIM_Cubic>,<at 0.1 0.0017 RCIM_Cubic>]}
      }})
    Particles.ColorFlare2:LinearColor=&MakeLinearColorFromVectorAndFloat(
     "Vector (RGB)"=&Multiply_Vector(
      A=&VectorFromFloat(Value=2)
      B=&MakeVectorFromLinearColorRGB(Color=<rgba 0.0017 0.7239 1 1>)
     )
     "Float (Alpha)"=&FloatFromCurve(FloatCurve=DataInterface() {propertyValues {Curve {keys [
          {arriveTangent=9.0725, interpMode=RCIM_Cubic, leaveTangent=9.0725, tangentMode=RCTM_User, time=0, value=0}
          <at 0.05 0.9847 RCIM_Cubic>
          {arriveTangent=-0.0405, interpMode=RCIM_Cubic, leaveTangent=-0.0405, tangentMode=RCTM_User, time=1, value=0}
         ]}}})
    )
   ) {isSetParameters=true, module=SetVariables_#5}
   /SetVariables(
    Particles.PositionFlare:Vector3f=&VectorFromCurve(VectorCurve=DataInterface() {propertyValues {
       XCurve {keys [
         <at 0.0009 -3.3 RCIM_Cubic>
         {arriveTangent=0.9114, interpMode=RCIM_Cubic, leaveTangent=0.9114, tangentMode=RCTM_User, time=0.8155, value=-2.0807}
         {arriveTangent=11.1923, interpMode=RCIM_Cubic, leaveTangent=11.1929, tangentMode=RCTM_User, time=1, value=1.5}
        ]}
       YCurve {keys [<at 0 0 RCIM_Cubic>]}
       ZCurve {keys [
         {arriveTangent=8.0617, interpMode=RCIM_Cubic, leaveTangent=8.0617, tangentMode=RCTM_User, time=0, value=15}
         {arriveTangent=104.9599, interpMode=RCIM_Cubic, leaveTangent=104.9599, tangentMode=RCTM_User, time=0.1892, value=27.259}
         {arriveTangent=18.9662, interpMode=RCIM_Cubic, leaveTangent=18.9662, tangentMode=RCTM_User, time=0.4418, value=60}
        ]}
      }})
    Particles.PositionFlare2:Vector3f=&VectorFromCurve(
     VectorCurve=DataInterface() {propertyValues {
       XCurve {keys [
         <at 0.0009 -3.3 RCIM_Cubic>
         {arriveTangent=2.2118, interpMode=RCIM_Cubic, leaveTangent=2.2118, tangentMode=RCTM_User, time=0.9, value=-2}
         {arriveTangent=11.1923, interpMode=RCIM_Cubic, leaveTangent=11.1929, tangentMode=RCTM_User, time=1, value=1.5}
        ]}
       YCurve {keys [<at 0 0>,<at 1 0>]}
       ZCurve {keys [
         <at 0.1335 18.6599 RCIM_Cubic>
         {arriveTangent=161.2398, interpMode=RCIM_Cubic, leaveTangent=161.2398, time=0.2796, value=78.0762}
         {arriveTangent=87.8052, interpMode=RCIM_Cubic, leaveTangent=87.8071, tangentMode=RCTM_User, time=0.7, value=110}
        ]}
      }}
     Scale Curve=<1 1 0.8>
    )
   ) {isSetParameters=true, module=SetVariables_#6}
   &SolveForcesAndVelocity(Clamp Velocity=false, Limit Acceleration=false)
   &GenerateDeathEvent(Gap Correction Amount=1.5) {enabled=false}
   &ScaleColor(
    Scale Alpha=&FloatFromCurve(FloatCurve=DataInterface() {propertyValues {Curve {keys [
         {interpMode=RCIM_Cubic, tangentMode=RCTM_User, time=0, value=2}
         {arriveTangent=-4.695, interpMode=RCIM_Cubic, leaveTangent=-4.695, tangentMode=RCTM_User, time=0.2, value=1}
         {arriveTangent=-3.3333, interpMode=RCIM_Cubic, leaveTangent=-3.3333, tangentMode=RCTM_User, time=0.3, value=0}
         {arriveTangent=2.5, interpMode=RCIM_Cubic, leaveTangent=2.5, tangentMode=RCTM_User, time=0.5, value=0}
         {arriveTangent=2, interpMode=RCIM_Cubic, leaveTangent=2, time=0.6, value=0}
         <at 1 0 RCIM_Cubic>
        ]}}})
    Scale RGB=&MakeVectorFromLinearColorRGB(Color=&ColorFromCurve(ColorCurve=DataInterface() {propertyValues {
        AlphaCurve {keys [<at 0 1 RCIM_Cubic>,<at 1 1 RCIM_Cubic>]}
        BlueCurve {keys [<at 0 1 RCIM_Cubic>,<at 0.5113 4>]}
        GreenCurve {keys [<at 0 1 RCIM_Cubic>,<at 0.5113 1.7231>]}
        RedCurve {keys [<at 0 1 RCIM_Cubic>,<at 0.5113 0.04>]}
       }}))
    ScaleA=true
    ScaleRGB=true
   )
  ]
  rendererClasses [&NiagaraMeshRendererProperties, &NiagaraSpriteRendererProperties]
  renderers [
   {rendererClass=&NiagaraMeshRendererProperties, rendererIndex=0}
   {rendererClass=&NiagaraSpriteRendererProperties, rendererIndex=1}
   {rendererClass=&NiagaraSpriteRendererProperties, rendererIndex=2}
   {rendererClass=&NiagaraSpriteRendererProperties, rendererIndex=3}
  ]
 }
 SheenMesh_Loop001 {
  bEnabled=false
  emitterUpdateScript [
   &EmitterState(
    Life Cycle Mode=::NewEnumerator1
    Loop Delay=&RandomRangeFloat(Minimum=0, Maximum=1, Evaluation Type=::NewEnumerator1)
    Loop Duration=&RandomRangeFloat(Minimum=1, Maximum=1)
    Recalculate Duration Each Loop=true
    UseLoopDelay=false
   )
   &SpawnBurst_Instantaneous(Spawn Count=1, Use Spawn Probability=false)
   &SpawnBurst_Instantaneous(Spawn Time=0.1, Use Spawn Probability=false) {module=SpawnBurst_Instantaneous001}
   &SpawnRate(SpawnRate=20, Use Spawn Probability=false) {enabled=false}
  ]
  particleSpawnScript [
   &Color(Color=&MakeLinearColorFromVectorAndFloat(
     "Vector (RGB)"=<1 0 0>
     "Float (Alpha)"=&ClampFloat(Float=&RandomRangeFloat(Minimum=0.4, Maximum=0.6, Evaluation Type=::NewEnumerator1))
    )) {enabled=false}
   /SetVariables(
    Particles.ColorFlare:LinearColor=&MultiplyLinearColor(Linear Color=$System.GlintColor, Scale Factor=<20 20 20 20>)
    Particles.PositionFlare:NiagaraPosition=$Particles.Position
   ) {isSetParameters=true, module=SetVariables_#2}
   &InitializeParticle(
    Color=<rgba 0.0348 14.4774 20 1>
    Lifetime=&RandomRangeFloat(Minimum=2, Maximum=2)
    Module.Write Color=true
    Module.Write Lifetime=true
    Module.Write Mass=true
    Module.Write Material Random=true
    Module.Write Position=false
    Module.Write Scale=true
    Module.Write SpriteRotation=false
    Module.Write SpriteSize=true
    Position=<0 -1 -2>
    Sprite Size=<1 2>
   )
   &FN_SpriteAlignmentAndFacing(SpriteAlignment:Vector3f=<0 0 1>, SpriteFacing:Vector3f=<1 0 0>)
  ]
  particleUpdateScript [
   &ParticleState()
   &ScaleSpriteSize(Scale Factor=&MakeVector2D(
     X=&FloatFromCurve(
      FloatCurve=DataInterface() {propertyValues {Curve {keys [
          {arriveTangent=7.953, interpMode=RCIM_Cubic, leaveTangent=7.953, tangentMode=RCTM_User, time=-0.0008, value=0.5}
          {arriveTangent=0.0662, interpMode=RCIM_Cubic, leaveTangent=0.0662, time=0.1, value=1.5834}
          {arriveTangent=-4.779, interpMode=RCIM_Cubic, leaveTangent=-4.779, time=0.2045, value=0.5136}
          {arriveTangent=-0.1355, interpMode=RCIM_Cubic, leaveTangent=-0.1355, tangentMode=RCTM_User, time=0.4, value=0.1497}
         ]}}}
      Scale Curve=10
     )
     Y=&FloatFromCurve(
      FloatCurve=DataInterface() {propertyValues {Curve {keys [
          {arriveTangent=7.953, interpMode=RCIM_Cubic, leaveTangent=7.953, tangentMode=RCTM_User, time=-0.0008, value=0.5}
          {arriveTangent=0.8674, interpMode=RCIM_Cubic, leaveTangent=0.8674, time=0.15, value=1.8998}
          {arriveTangent=-6.9225, interpMode=RCIM_Cubic, leaveTangent=-6.9225, time=0.2096, value=0.6825}
          {arriveTangent=-0.4678, interpMode=RCIM_Cubic, leaveTangent=-0.4678, tangentMode=RCTM_User, time=0.4, value=0.1692}
         ]}}}
      Scale Curve=10
     )
    ))
   /SetVariables(Particles.MeshIndex=&RandomRangeInt(Minimum=0, Maximum=1)) {isSetParameters=true}
   /SetVariables(Particles.SpriteSizeFlare2:Vector2f=&MakeVector2D(
     X=&FloatFromCurve(
      FloatCurve=DataInterface() {propertyValues {Curve {keys [
          {arriveTangent=7.953, interpMode=RCIM_Cubic, leaveTangent=7.953, tangentMode=RCTM_User, time=-0.0008, value=0.5}
          {interpMode=RCIM_Cubic, leaveTangent=-0.7143, time=0.225, value=1.8998}
          {arriveTangent=-0.1355, interpMode=RCIM_Cubic, leaveTangent=-0.1355, tangentMode=RCTM_User, time=0.6, value=0.1497}
         ]}}}
      Scale Curve=10
     )
     Y=&FloatFromCurve(
      FloatCurve=DataInterface() {propertyValues {Curve {keys [
          {arriveTangent=7.953, interpMode=RCIM_Cubic, leaveTangent=7.953, tangentMode=RCTM_User, time=-0.0008, value=0.5}
          {interpMode=RCIM_Cubic, leaveTangent=-0.6804, time=0.225, value=1.8998}
          {arriveTangent=-0.4678, interpMode=RCIM_Cubic, leaveTangent=-0.4678, tangentMode=RCTM_User, time=0.6, value=0.1692}
         ]}}}
      Scale Curve=15
     )
    )) {isSetParameters=true, module=SetVariables_#3}
   &Color(Color=&MakeLinearColorFromVectorAndFloat("Vector (RGB)"=<0.4208 0.6361 1>, "Float (Alpha)"=&ClampFloat(Float=0.8))) {enabled=false, module=Color001}
   &DynamicMaterialParameters(
    Index 0 Param 1=&FloatFromCurve(FloatCurve=DataInterface() {propertyValues {Curve {keys [{arriveTangent=2.513, interpMode=RCIM_Cubic, time=0, value=0.2}, <at 0.3 1.8>]}}})
    Index 0 Param 2=1
    Index 0 Param 3=&FloatFromCurve(
     FloatCurve=DataInterface() {propertyValues {Curve {keys [{interpMode=RCIM_Cubic, tangentMode=RCTM_User, time=0.2, value=0}, <at 0.4 2 RCIM_Cubic>, <at 1 0 RCIM_Cubic>]}}}
     Scale Curve=0.2
    )
    Index 0 Param 4=&FloatFromCurve(FloatCurve=DataInterface() {propertyValues {Curve {keys [
         {interpMode=RCIM_Cubic, tangentMode=RCTM_User, time=0, value=0}
         {arriveTangent=1.2104, interpMode=RCIM_Cubic, leaveTangent=1.2104, tangentMode=RCTM_User, time=0.1, value=0.05}
         {arriveTangent=1.6138, interpMode=RCIM_Cubic, leaveTangent=1.6138, tangentMode=RCTM_User, time=0.3, value=0.4}
         <at 1 1>
        ]}}})
    Module.Param0WriteEnabled=true
    Module.Param1WriteEnabled=true
    Module.Param2WriteEnabled=true
    Module.Param3WriteEnabled=true
   )
   &CameraOffset(Camera Offset Amount=1)
   /SetVariables(Particles.CameraOffsetFlare=2) {isSetParameters=true, module=SetVariables_#4}
   /SetVariables(
    Particles.ColorFlare:LinearColor=&ColorFromCurve(ColorCurve=DataInterface() {propertyValues {
       AlphaCurve {keys [<at 0.2801 1 RCIM_Cubic>,<at 0.5537 0 RCIM_Cubic>]}
       BlueCurve {keys [<at 0 1 RCIM_Cubic>,<at 0.1 1 RCIM_Cubic>]}
       GreenCurve {keys [<at 0 1 RCIM_Cubic>,<at 0.1 0.7239 RCIM_Cubic>]}
       RedCurve {keys [<at 0 1 RCIM_Cubic>,<at 0.1 0.0017 RCIM_Cubic>]}
      }})
    Particles.ColorFlare2:LinearColor=&MakeLinearColorFromVectorAndFloat(
     "Vector (RGB)"=&Multiply_Vector(
      A=&VectorFromFloat(Value=20)
      B=&MakeVectorFromLinearColorRGB(Color=<rgba 0.0017 0.7239 1 1>)
     )
     "Float (Alpha)"=&FloatFromCurve(FloatCurve=DataInterface() {propertyValues {Curve {keys [
          {arriveTangent=9.0725, interpMode=RCIM_Cubic, leaveTangent=9.0725, tangentMode=RCTM_User, time=0, value=0}
          <at 0.05 0.9847 RCIM_Cubic>
          {arriveTangent=-0.0405, interpMode=RCIM_Cubic, leaveTangent=-0.0405, tangentMode=RCTM_User, time=1, value=0}
         ]}}})
    )
   ) {isSetParameters=true, module=SetVariables_#5}
   /SetVariables(
    Particles.PositionFlare:Vector3f=&VectorFromCurve(VectorCurve=DataInterface() {propertyValues {
       XCurve {keys [
         <at 0.0009 -3.3 RCIM_Cubic>
         {arriveTangent=0.9114, interpMode=RCIM_Cubic, leaveTangent=0.9114, tangentMode=RCTM_User, time=0.8155, value=-2.0807}
         {arriveTangent=11.1923, interpMode=RCIM_Cubic, leaveTangent=11.1929, tangentMode=RCTM_User, time=1, value=1.5}
        ]}
       YCurve {keys [<at 0 0 RCIM_Cubic>]}
       ZCurve {keys [
         {arriveTangent=8.0617, interpMode=RCIM_Cubic, leaveTangent=8.0617, tangentMode=RCTM_User, time=0, value=15}
         {arriveTangent=104.9599, interpMode=RCIM_Cubic, leaveTangent=104.9599, tangentMode=RCTM_User, time=0.1892, value=27.259}
         {arriveTangent=18.9662, interpMode=RCIM_Cubic, leaveTangent=18.9662, tangentMode=RCTM_User, time=0.4418, value=60}
        ]}
      }})
    Particles.PositionFlare2:Vector3f=&VectorFromCurve(
     VectorCurve=DataInterface() {propertyValues {
       XCurve {keys [
         <at 0.0009 -3.3 RCIM_Cubic>
         {arriveTangent=2.2118, interpMode=RCIM_Cubic, leaveTangent=2.2118, tangentMode=RCTM_User, time=0.9, value=-2}
         {arriveTangent=11.1923, interpMode=RCIM_Cubic, leaveTangent=11.1929, tangentMode=RCTM_User, time=1, value=1.5}
        ]}
       YCurve {keys [<at 0 0>,<at 1 0>]}
       ZCurve {keys [
         <at 0.1335 18.6599 RCIM_Cubic>
         {arriveTangent=161.2398, interpMode=RCIM_Cubic, leaveTangent=161.2398, time=0.2796, value=78.0762}
         {arriveTangent=87.8052, interpMode=RCIM_Cubic, leaveTangent=87.8071, tangentMode=RCTM_User, time=0.7, value=110}
        ]}
      }}
     Scale Curve=<1 1 0.8>
    )
   ) {isSetParameters=true, module=SetVariables_#6}
   &SolveForcesAndVelocity(Clamp Velocity=false, Limit Acceleration=false)
   &GenerateDeathEvent(Gap Correction Amount=1.5)
   &ScaleColor(
    Scale Alpha=&FloatFromCurve(FloatCurve=DataInterface() {propertyValues {Curve {keys [
         {interpMode=RCIM_Cubic, tangentMode=RCTM_User, time=0, value=2}
         {arriveTangent=-4.695, interpMode=RCIM_Cubic, leaveTangent=-4.695, tangentMode=RCTM_User, time=0.2, value=1}
         {arriveTangent=-3.3333, interpMode=RCIM_Cubic, leaveTangent=-3.3333, tangentMode=RCTM_User, time=0.3, value=0}
         {arriveTangent=2.5, interpMode=RCIM_Cubic, leaveTangent=2.5, tangentMode=RCTM_User, time=0.5, value=0}
         {arriveTangent=2, interpMode=RCIM_Cubic, leaveTangent=2, time=0.6, value=1}
         <at 1 1 RCIM_Cubic>
        ]}}})
    ScaleA=true
    ScaleRGB=true
   )
  ]
  rendererClasses [&NiagaraMeshRendererProperties, &NiagaraSpriteRendererProperties]
  renderers [
   {rendererClass=&NiagaraMeshRendererProperties, rendererIndex=0}
   {rendererClass=&NiagaraSpriteRendererProperties, rendererIndex=1}
   {rendererClass=&NiagaraSpriteRendererProperties, rendererIndex=2}
   {rendererClass=&NiagaraSpriteRendererProperties, rendererIndex=3}
  ]
 }
}
summary {
 systemName=NS_Neon_Cat_Tech_Idle_1h_R
 emitters [
  {emitterName=SheenMesh_Loop, rendererClasses [&NiagaraMeshRendererProperties, &NiagaraSpriteRendererProperties]}
  {
   bEnabled=false
   emitterName=SheenMesh_Loop001
   rendererClasses [&NiagaraMeshRendererProperties, &NiagaraSpriteRendererProperties]
  }
 ]
 userVariables [User.Swing_SK_Mesh {
   type=NiagaraDataInterfaceSkeletalMesh
   defaultValue {
    struct=NiagaraExt_VariableValue_DataInterface
    value {dataInterface=&self:NiagaraDataInterfaceSkeletalMesh_0}
   }
  }]
}
