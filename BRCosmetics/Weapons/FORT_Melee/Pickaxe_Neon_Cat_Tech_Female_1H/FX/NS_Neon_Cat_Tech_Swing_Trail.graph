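# Niagara system for the Neon Cat Tech pickaxe swing trail, a short blue-cyan effect built from three active emitters.
# AirTrail draws three animation-trail ribbon renderers with two color curves, while BladeGlow bursts one mesh-and-sprite glow with curve-animated flare positions and sizes.
# FN_Attached_Ribbon_BETA samples User.SkeletalMesh, using a scratch-pad bone lock that holds locked particles at zero age and pinned to the mesh.
# Trail_Swing, Light and Empty emitters are disabled.
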NiagaraSystem /BRCosmetics/Weapons/FORT_Melee/Pickaxe_Neon_Cat_Tech_Female_1H/FX/NS_Neon_Cat_Tech_Swing_Trail
dependencies {
 usedDataInterfaces [&NiagaraDataInterfaceColorCurve, &NiagaraDataInterfaceCurve, &NiagaraDataInterfaceVectorCurve]
 usedDynamicInputs []
 usedModules [&SystemState]
}
emitters {
 AirTrail {
  emitterUpdateScript [
   &EmitterState(Life Cycle Mode=::NewEnumerator1, Loop Behavior=::NewEnumerator1, Loop Duration=0.2, UseLoopDelay=false)
   &FN_SpawnAnimTrail_v2()
  ]
  particleSpawnScript [
   /Game/Effects/Niagara/Modules/AnimTrail/FN_InitializeAnimTrail/FN_InitializeAnimTrailBoneIndecies(First Bone or Socket Index=1)
   /Game/Effects/Niagara/Modules/AnimTrail/FN_InitializeAnimTrail/FN_InitializeAnimTrail(AnimTrailWidth=1, Color=<rgba 0.0031 0.0212 0.05 1>, Lifetime=0.15)
   /SetVariables() {isSetParameters=true}
  ]
  particleUpdateScript [
   &SolveForcesAndVelocity(Clamp Velocity=false, Limit Acceleration=false)
   &ScaleRibbonWidth(Ribbon Width Scale=&FloatFromCurve(FloatCurve=DataInterface() {propertyValues {Curve {keys [
         {arriveTangent=-1.2579, interpMode=RCIM_Cubic, time=0.4, value=1}
         {arriveTangent=-0.9664, interpMode=RCIM_Cubic, leaveTangent=-0.9664, tangentMode=RCTM_User, time=1, value=0.5}
        ]}}}))
   &FN_UpdateAnimTrail_v2()
   &DynamicMaterialParameters(
    Index 0 Param 1=&UniformRangedFloat(Minimum=0, Maximum=1)
    Index 0 Param 2=&UniformRangedFloat(Minimum=0, Maximum=1)
    Index 0 Param 3=1
    Index 0 Param 4=1
    Module.Param0WriteEnabled=true
    Module.Param1WriteEnabled=true
    Module.Param2WriteEnabled=true
    Module.Param3WriteEnabled=true
   )
   &Color(Color=&MultiplyLinearColor(
     Linear Color=&ColorFromCurve(ColorCurve=DataInterface() {propertyValues {
        AlphaCurve {keys [<at 0 5 RCIM_Cubic>,<at 1 1 RCIM_Cubic>]}
        BlueCurve {keys [<at 0 1 RCIM_Cubic>,<at 0.3089 1>]}
        GreenCurve {keys [<at 0 0.7234 RCIM_Cubic>,<at 0.3089 0.5111>]}
        RedCurve {keys [<at 0 0 RCIM_Cubic>,<at 0.3089 0>]}
       }})
     Scale Factor=&MakeVector4_FromFloat(Float=2)
    ))
   /SetVariables(Particles.Color2=&MultiplyLinearColor(
     Linear Color=&ColorFromCurve(ColorCurve=DataInterface() {propertyValues {
        AlphaCurve {keys [<at 0.4651 5 RCIM_Cubic>,<at 1 1 RCIM_Cubic>]}
        BlueCurve {keys [<at 0 1 RCIM_Cubic>,<at 0.9446 1>]}
        GreenCurve {keys [<at 0 0.3673 RCIM_Cubic>,<at 0.9446 0.3688>]}
        RedCurve {keys [<at 0 0.006 RCIM_Cubic>,<at 0.9446 0>]}
       }})
     Scale Factor=&MakeVector4_FromFloat(Float=5)
    )) {isSetParameters=true}
  ]
  rendererClasses [&NiagaraRibbonRendererProperties]
  renderers [
   {rendererClass=&NiagaraRibbonRendererProperties}
   {rendererClass=&NiagaraRibbonRendererProperties, rendererIndex=1}
   {rendererClass=&NiagaraRibbonRendererProperties, rendererIndex=2}
  ]
 }
 BladeGlow {
  emitterUpdateScript [
   &EmitterState(
    Life Cycle Mode=::NewEnumerator1
    Loop Behavior=::NewEnumerator1
    Loop Delay=&RandomRangeFloat(Minimum=0, Maximum=1, Evaluation Type=::NewEnumerator1)
    Loop Duration=1
    UseLoopDelay=false
   )
   &SpawnBurst_Instantaneous(Spawn Count=1, Use Spawn Probability=false)
   &SpawnRate(SpawnRate=20, Use Spawn Probability=false) {enabled=false}
  ]
  particleSpawnScript [
   &Color(Color=&MakeLinearColorFromVectorAndFloat(
     "Vector (RGB)"=<1 0 0>
     "Float (Alpha)"=&ClampFloat(Float=&RandomRangeFloat(Minimum=0.4, Maximum=0.6, Evaluation Type=::NewEnumerator1))
    )) {enabled=false}
   /SetVariables(
    Particles.ColorFlare:LinearColor=&MultiplyLinearColor(Linear Color=$System.GlintColor, Scale Factor=<20 20 20 20>)
    Particles.PositionFlare:NiagaraPosition=$Particles.Position
   ) {isSetParameters=true}
   /Niagara/Modules/Spawn/Initialization/InitializeParticle(
    Color=<rgba 0.6 14.8267 20 1>
    Lifetime=0.3
    Module.Write Color=true
    Module.Write Lifetime=true
    Module.Write Mass=true
    Module.Write Material Random=true
    Module.Write Position=false
    Module.Write Scale=true
    Module.Write SpriteRotation=false
    Module.Write SpriteSize=true
    Position=<0 -1 -2>
    Sprite Size=<1 2>
   )
   /NiagaraFN/Sprite/FN_SpriteAlignmentAndFacing(SpriteAlignment:Vector3f=<0 0 1>, SpriteFacing:Vector3f=<1 0 0>)
  ]
  particleUpdateScript [
   &ParticleState()
   &ScaleSpriteSize(Scale Factor=&MakeVector2D(
     X=&FloatFromCurve(
      FloatCurve=DataInterface() {propertyValues {Curve {keys [
          {arriveTangent=7.953, interpMode=RCIM_Cubic, leaveTangent=7.953, tangentMode=RCTM_User, time=-0.0008, value=0.5}
          {interpMode=RCIM_Cubic, leaveTangent=-0.7143, time=0.225, value=1.8998}
          {arriveTangent=-0.1355, interpMode=RCIM_Cubic, leaveTangent=-0.1355, tangentMode=RCTM_User, time=0.6, value=0.1497}
         ]}}}
      Scale Curve=10
     )
     Y=&FloatFromCurve(
      FloatCurve=DataInterface() {propertyValues {Curve {keys [
          {arriveTangent=7.953, interpMode=RCIM_Cubic, leaveTangent=7.953, tangentMode=RCTM_User, time=-0.0008, value=0.5}
          {interpMode=RCIM_Cubic, leaveTangent=-0.6804, time=0.225, value=1.8998}
          {arriveTangent=-0.4678, interpMode=RCIM_Cubic, leaveTangent=-0.4678, tangentMode=RCTM_User, time=0.6, value=0.1692}
         ]}}}
      Scale Curve=10
     )
    ))
   /SetVariables(Particles.SpriteSizeFlare2:Vector2f=&MakeVector2D(
     X=&FloatFromCurve(
      FloatCurve=DataInterface() {propertyValues {Curve {keys [
          {arriveTangent=-2.2776, interpMode=RCIM_Cubic, leaveTangent=-2.2776, tangentMode=RCTM_User, time=0, value=1}
          {arriveTangent=-1, interpMode=RCIM_Cubic, leaveTangent=-1, time=0.3, value=1.75}
          <at 1 0 RCIM_Cubic>
         ]}}}
      Scale Curve=3
     )
     Y=&FloatFromCurve(
      FloatCurve=DataInterface() {propertyValues {Curve {keys [<at 0 1 RCIM_Cubic>, {arriveTangent=-1.3497, interpMode=RCIM_Cubic, leaveTangent=-1.3497, tangentMode=RCTM_User, time=1, value=0.5}]}}}
      Scale Curve=20
     )
    )) {isSetParameters=true}
   &Color(Color=&MakeLinearColorFromVectorAndFloat("Vector (RGB)"=<0.4208 0.6361 1>, "Float (Alpha)"=&ClampFloat(Float=0.8))) {enabled=false, module=Color001}
   &DynamicMaterialParameters(
    Index 0 Param 1=&FloatFromCurve(FloatCurve=DataInterface() {propertyValues {Curve {keys [{arriveTangent=2.513, interpMode=RCIM_Cubic, time=0, value=0.2}, <at 0.3 1.8>]}}})
    Index 0 Param 2=1
    Index 0 Param 3=&FloatFromCurve(
     FloatCurve=DataInterface() {propertyValues {Curve {keys [{interpMode=RCIM_Cubic, tangentMode=RCTM_User, time=0.2, value=0}, <at 0.4 2 RCIM_Cubic>, <at 1 0 RCIM_Cubic>]}}}
     Scale Curve=0.2
    )
    Index 0 Param 4=&FloatFromCurve(FloatCurve=DataInterface() {propertyValues {Curve {keys [
         {interpMode=RCIM_Cubic, tangentMode=RCTM_User, time=0, value=0}
         {arriveTangent=1.2104, interpMode=RCIM_Cubic, leaveTangent=1.2104, tangentMode=RCTM_User, time=0.7, value=0.25}
         <at 1 1>
        ]}}})
    Module.Param0WriteEnabled=true
    Module.Param1WriteEnabled=true
    Module.Param2WriteEnabled=true
    Module.Param3WriteEnabled=true
   )
   &CameraOffset(Camera Offset Amount=1)
   /SetVariables(Particles.CameraOffsetFlare=2) {isSetParameters=true}
   /SetVariables(Particles.ColorFlare:LinearColor=&ColorFromCurve(ColorCurve=DataInterface() {propertyValues {
       AlphaCurve {keys [<at 0.2801 1 RCIM_Cubic>,<at 0.5537 0 RCIM_Cubic>]}
       BlueCurve {keys [<at 0 1 RCIM_Cubic>,<at 0.2942 0.8714 RCIM_Cubic>]}
       GreenCurve {keys [<at 0 1 RCIM_Cubic>,<at 0.2942 0.6544 RCIM_Cubic>]}
       RedCurve {keys [<at 0 1 RCIM_Cubic>,<at 0.2942 0.0871 RCIM_Cubic>]}
      }})) {isSetParameters=true}
   /SetVariables(
    Particles.PositionFlare:Vector3f=&VectorFromCurve(VectorCurve=DataInterface() {propertyValues {
       XCurve {keys [
         <at 0.0009 -3.3 RCIM_Cubic>
         {interpMode=RCIM_Cubic, leaveTangent=9.1028, time=0.9, value=-3.3}
         {arriveTangent=11.1923, interpMode=RCIM_Cubic, leaveTangent=11.1929, tangentMode=RCTM_User, time=1, value=1.5}
        ]}
       YCurve {keys [<at 0 0>,<at 1 0>]}
       ZCurve {keys [
         <at 0 15 RCIM_Cubic>
         {arriveTangent=21.9422, interpMode=RCIM_Cubic, leaveTangent=21.9415, tangentMode=RCTM_User, time=0.2404, value=20}
         {arriveTangent=87.8052, interpMode=RCIM_Cubic, leaveTangent=87.8071, tangentMode=RCTM_User, time=0.7, value=70}
        ]}
      }})
    Particles.PositionFlare2:Vector3f=&MakeVector(
     X=1
     Y=&FloatFromCurve(FloatCurve=DataInterface() {propertyValues {Curve {keys [
          {arriveTangent=134.5043, interpMode=RCIM_Cubic, leaveTangent=134.5037, tangentMode=RCTM_User, time=0, value=15}
          {arriveTangent=33.5313, interpMode=RCIM_Cubic, leaveTangent=33.5312, tangentMode=RCTM_User, time=0.5, value=30}
         ]}}})
     Z=69
    )
   ) {isSetParameters=true}
   &SolveForcesAndVelocity(Clamp Velocity=false, Limit Acceleration=false)
   &ScaleColor(
    Scale Alpha=&FloatFromCurve(FloatCurve=DataInterface() {propertyValues {Curve {keys [
         {interpMode=RCIM_Cubic, tangentMode=RCTM_User, time=0, value=2}
         {arriveTangent=-4.695, interpMode=RCIM_Cubic, leaveTangent=-4.695, tangentMode=RCTM_User, time=0.2, value=1}
         {arriveTangent=-3.3333, interpMode=RCIM_Cubic, leaveTangent=-3.3333, tangentMode=RCTM_User, time=0.3, value=0}
         {arriveTangent=2.5, interpMode=RCIM_Cubic, leaveTangent=2.5, tangentMode=RCTM_User, time=0.5, value=0}
         {arriveTangent=2, interpMode=RCIM_Cubic, leaveTangent=2, time=0.6, value=1}
         <at 1 1 RCIM_Cubic>
        ]}}})
    ScaleA=true
    ScaleRGB=true
   )
  ]
  rendererClasses [&NiagaraMeshRendererProperties, &NiagaraSpriteRendererProperties]
  renderers [
   {rendererClass=&NiagaraMeshRendererProperties, rendererIndex=0}
   {rendererClass=&NiagaraSpriteRendererProperties, rendererIndex=1}
   {rendererClass=&NiagaraSpriteRendererProperties, rendererIndex=2}
  ]
 }
 Empty {
  bEnabled=false
  emitterUpdateScript [&EmitterState()]
  particleSpawnScript [/Niagara/Modules/Spawn/Initialization/V2/InitializeParticle(Color Mode=::NewEnumerator0, Lifetime=5, UsePositionOffset=false, Write Material Random=true)]
  particleUpdateScript [&ParticleState()]
  rendererClasses [&NiagaraSpriteRendererProperties]
  renderers [{rendererClass=&NiagaraSpriteRendererProperties}]
 }
 FN_Attached_Ribbon_BETA {
  emitterSpawnScript [/SetVariables(Emitter.Number of Trails=1) {isSetParameters=true}]
  emitterUpdateScript [
   &EmitterState(Life Cycle Mode=::NewEnumerator1, Loop Behavior=::NewEnumerator1, Loop Delay=0.03, Loop Duration=1, UseLoopDelay=true)
   &SpawnBurst_Instantaneous(
    Spawn Count=$Emitter.Initial.Number of Trails
    Spawn Group=-1
    Spawn Probability=HlslExpression() {hlslExpression="Emitter.LoopCount == 0"}
    Use Spawn Probability=true
   )
   &SpawnRate(
    SpawnRate=&Multiply_Float(A=60, B=&MakeFloatFromInt(Int=$Emitter.Number of Trails))
    Use Spawn Probability=false
   )
  ]
  particleSpawnScript [
   &FN_SetSpawnGroup(Number of Spawn Groups:NiagaraInt32=$Emitter.Initial.Number of Trails)
   &self:FN_Attached_Ribbon_BETA.NiagaraScratchPadContainer_1.LockParticlesToBones_11() {module=LockParticlesToBones}
   &InitializeRibbon(
    Color=<rgba 1 1 1 2>
    Lifetime=0.12
    Module.Write Color=true
    Module.Write Lifetime=true
    Module.Write Mass=true
    Module.Write Material Random=true
    Module.Write Position=false
    Module.Write RibbonTwist=false
    Module.Write RibbonWidth=true
    Ribbon Width=6
   )
   &FN_FlipRibbonV() {enabled=false}
   /SetVariables() {isSetParameters=true}
   &SkeletalMeshLocation(
    Inherit Velocity=false
    Interpolate Mesh Sampling=false
    MaskSpawnGroup=false
    Mesh Sampling Type=::NewEnumerator5
    Sampled Position Offset=$Particles.Offset
    Sampled Position Offset Coordinate Space:ENiagaraMeshTransforms=::NewEnumerator3
    Skeletal Mesh=$User.SkeletalMesh
    Socket Sampling Mode=::NewEnumerator5
   )
   &AddVelocity(Velocity=<0 0 100>) {enabled=false}
  ]
  particleUpdateScript [
   &ParticleState()
   /SetVariables(Particles.Age:NiagaraFloat=&MakeCustomFloatFromBool(Bool=$Particles.Locked, True Float=0, False Float=$Particles.Age)) {isSetParameters=true}
   /SetVariables(Particles.NormalizedAge:NiagaraFloat=&MakeCustomFloatFromBool(Bool=$Particles.Locked, True Float=0, False Float=$Particles.NormalizedAge)) {isSetParameters=true}
   &SkeletalMeshLocation(
    Apply To Particle Position=false
    Inherit Velocity=false
    Interpolate Mesh Sampling=false
    MaskSpawnGroup=false
    Mesh Sampling Type=::NewEnumerator5
    Sampled Position Offset Coordinate Space:ENiagaraMeshTransforms=::NewEnumerator3
    Skeletal Mesh=$User.SkeletalMesh
    Socket Sampling Mode=::NewEnumerator5
    Use Sampled Mesh Attribute as Ribbon ID=false
   ) {module=SkeletalMeshLocation001}
   &SolveForcesAndVelocity(Clamp Velocity=false, Limit Acceleration=false)
   /SetVariables(Particles.RenderPosition:Vector3f=&MakeCustomVectorFromBool(
     Bool=$Particles.Locked
     True Vector=&Add_Vector(
      A=$Output.SkeletalMeshLocation001.SampledPosition
      B=&RotateVectorByQuaternion(Vector To Rotate=$Particles.Offset, Quaternion=$Output.SkeletalMeshLocation001.SampledRotation)
     )
     False Vector=&MakeCustomVectorFromBool(
      Bool=$Emitter.LocalSpace
      True Vector=&Add_Vector(A=$Output.SkeletalMeshLocation001.SampledPosition, B=Linked() {linkedVariable {name=Particles.Position, type=NiagaraPosition}})
      False Vector=Linked() {linkedVariable {name=Particles.Position, type=NiagaraPosition}}
     )
    )) {isSetParameters=true}
   &Color(Scale Alpha=&FloatFromCurve(
     FloatCurve=DataInterface() {propertyValues {Curve {keys [<at 0.0009 0.9773 RCIM_Cubic>, {arriveTangent=-0.9781, interpMode=RCIM_Cubic, leaveTangent=-0.9781, time=0.3498, value=0.9932}, <at 1 0>]}}}
     Scale Curve=&FloatFromCurve(
      FloatCurve=DataInterface() {propertyValues {Curve {keys [
          <at 0.1592 0.0035 RCIM_Cubic>
          {arriveTangent=-0.0523, interpMode=RCIM_Cubic, leaveTangent=-0.0523, time=0.4157, value=0.9671}
          <at 0.664 -0.0229>
         ]}}}
      CurveIndex=$Emitter.Age
     )
    )) {enabled=false}
   &FN_RibbonMaterial(
    Opacity Pan U Speed=0.9
    Scale Alpha:NiagaraFloat=&FloatFromCurve(
     FloatCurve=DataInterface() {propertyValues {Curve {keys [<at 0.0009 0.9773 RCIM_Cubic>, {arriveTangent=-0.9781, interpMode=RCIM_Cubic, leaveTangent=-0.9781, time=0.3498, value=0.9932}, <at 1 0>]}}}
     Scale Curve=&FloatFromCurve(
      FloatCurve=DataInterface() {propertyValues {Curve {keys [<at 0.0609 0 RCIM_Cubic>, {arriveTangent=0.0268, interpMode=RCIM_Cubic, leaveTangent=0.0268, time=0.1693, value=0.9588}, <at 0.2946 0.0063>]}}}
      CurveIndex=$Emitter.Age
     )
    )
   ) {enabled=false}
   &ScaleColor(
    Linear Color Curve=DataInterface() {propertyValues {
      AlphaCurve {keys [<at 0.2055 0>,<at 0.2994 1>,<at 0.349 1>,<at 1 0>]}
      BlueCurve {keys [<at 0.2338 0.02>,<at 0.3171 0.8714>,<at 0.5775 1>]}
      GreenCurve {keys [<at 0.2338 0.02>,<at 0.3171 0.6308>,<at 0.5775 0.4735>]}
      RedCurve {keys [<at 0.2338 0.02>,<at 0.3171 0.0015>,<at 0.5775 0.0844>]}
     }}
    Scale Mode=::NewEnumerator2
   )
  ]
  rendererClasses [&NiagaraRibbonRendererProperties]
  renderers [{rendererClass=&NiagaraRibbonRendererProperties}]
 }
 Light {
  bEnabled=false
  emitterUpdateScript [
   &EmitterState(Life Cycle Mode=::NewEnumerator1, Loop Behavior=::NewEnumerator1, Loop Duration=1, UseLoopDelay=false)
   &SpawnBurst_Instantaneous(Spawn Count=1, Use Spawn Probability=false)
  ]
  particleSpawnScript [
   &InitializeParticle(
    Color=<rgba 5 0.4 3.6967 1>
    Lifetime=0.5
    Sprite Size Mode=::NewEnumerator3
    Uniform Sprite Size=30
    UsePositionOffset=false
    Write Material Random=true
   ) {module=InitializeParticle001}
   /SetVariables(Particles.Light_Falloff=1.8) {isSetParameters=true}
   /SetVariables(Particles.LightRadius=72) {isSetParameters=true}
  ]
  particleUpdateScript [
   &ParticleState()
   &SolveForcesAndVelocity(Clamp Velocity=false, Limit Acceleration=false)
   &ScaleColor(
    Scale Alpha=&FloatFromCurve(FloatCurve=DataInterface() {propertyValues {Curve {keys [
         {tangentMode=RCTM_User, time=0, value=1}
         {arriveTangent=-0.0405, interpMode=RCIM_Cubic, leaveTangent=-0.0405, tangentMode=RCTM_User, time=0.7009, value=1}
         {arriveTangent=-8.0089, interpMode=RCIM_Cubic, leaveTangent=-8.0089, tangentMode=RCTM_User, time=1, value=0}
        ]}}})
    Scale RGB=&Multiply_Vector(A=<1 1 1>, B=&VectorFromFloat(Value=1))
    ScaleA=true
    ScaleRGB=true
   )
   &SkeletalMeshLocation(
    Apply Sampled Orientation to Mesh Particles=true
    Inherit Velocity=false
    Interpolate Mesh Sampling=false
    MaskSpawnGroup=false
    Mesh Sampling Type=::NewEnumerator5
    Sampled Position Offset=<0 -10 0>
    Sampled Position Offset Coordinate Space:ENiagaraMeshTransforms=::NewEnumerator3
    Skeletal Mesh=DataInterface() {propertyValues {
      DefaultMesh=null
      ExcludeBoneName=root
      OptionalComponentClass=null
      PreviewMesh=/BRCosmetics/Weapons/FORT_Melee/Pickaxe_Crazy_Eight_Tech_Male_1H/Meshes/Crazy_Eight_Tech_Male_1h_R
      SkinningMode=SkinOnTheFly
      SoftSourceActor=null
      SourceMode=Default
      UvSetIndex=0
      WholeMeshLOD=-1
      bExcludeBone=false
      bReadDeformedGeometry=true
      bRequireCurrentFrameData=true
      ComponentTags []
      FilteredBones []
      FilteredSockets [AnimTrail_01,AnimTrail_02]
      MeshUserParameter {parameter {name=None, typeDefHandle {registeredTypeIndex=1}, varData []}}
      SamplingRegions []
     }}
    Socket Sampling Mode=::NewEnumerator5
   ) {module=SkeletalMeshLocation001}
  ]
  rendererClasses [&NiagaraLightRendererProperties]
  renderers [{rendererClass=&NiagaraLightRendererProperties, rendererIndex=0}]
 }
 Trail_Swing {
  bEnabled=false
  emitterUpdateScript [
   &EmitterState(Life Cycle Mode=::NewEnumerator1, Loop Behavior=::NewEnumerator1, Loop Delay=0.05, Loop Duration=0.35, UseLoopDelay=true)
   &SpawnRate(SpawnRate=45, Use Spawn Probability=false)
  ]
  particleSpawnScript [
   &InitializeRibbon(
    Color=<rgba 0.0348 14.4774 20 1>
    Lifetime=0.12
    Module.Write Color=true
    Module.Write Lifetime=true
    Module.Write Mass=true
    Module.Write Material Random=true
    Module.Write Position=false
    Module.Write RibbonTwist=false
    Module.Write RibbonWidth=true
    Ribbon Width=12
   )
   /SetVariables(Particles.RibbonFacing=<1 0 0>) {isSetParameters=true}
   &SkeletalMeshLocation(
    Bone / Socket Index=1
    Inherit Velocity=false
    MaskSpawnGroup=false
    Mesh Sampling Type=::NewEnumerator5
    Sampled Position Offset=<0 0 30>
    Sampled Position Offset Coordinate Space:ENiagaraMeshTransforms=::NewEnumerator3
    Skeletal Mesh=$User.SkeletalMesh
    Socket Sampling Mode=::NewEnumerator5
   )
  ]
  particleUpdateScript [
   &ParticleState()
   &SolveForcesAndVelocity(Clamp Velocity=false, Limit Acceleration=false)
   &ScaleColor(
    Scale Alpha=&Multiply_Float(
     A=&FloatFromCurve(FloatCurve=DataInterface() {propertyValues {Curve {keys [
          {interpMode=RCIM_Cubic, leaveTangent=-2.3646, tangentMode=RCTM_Break, time=0, value=1}
          {interpMode=RCIM_Cubic, tangentMode=RCTM_Break, time=1, value=0}
         ]}}})
     B=&Multiply_Float(
      A=&FloatFromCurve(FloatCurve=DataInterface() {propertyValues {Curve {keys [
           {arriveTangent=9.0725, interpMode=RCIM_Cubic, leaveTangent=9.0725, tangentMode=RCTM_User, time=0, value=0}
           {arriveTangent=-2.5249, interpMode=RCIM_Cubic, leaveTangent=-2.5249, tangentMode=RCTM_User, time=0.1408, value=0.9847}
           <at 0.4 0.2>
           {arriveTangent=-0.0405, interpMode=RCIM_Cubic, leaveTangent=-0.0405, tangentMode=RCTM_User, time=1, value=0}
          ]}}})
      B=&FNTimeOfDayScalar(Night Scale=1, Morning Scale=0.3, Evening Scale=0.5)
     )
    )
    Scale RGB=&MakeVectorFromLinearColorRGB(Color=&ColorFromCurve(ColorCurve=DataInterface() {propertyValues {
        AlphaCurve {keys [<at 0 1 RCIM_Cubic>]}
        BlueCurve {keys [<at 0 1 RCIM_Cubic>,<at 0.4969 0.2 RCIM_Cubic>]}
        GreenCurve {keys [<at 0 1 RCIM_Cubic>,<at 0.4969 0.1002 RCIM_Cubic>]}
        RedCurve {keys [<at 0 1 RCIM_Cubic>,<at 0.4969 0.0003 RCIM_Cubic>]}
       }}))
    ScaleA=true
    ScaleRGB=true
   )
   &ScaleRibbonWidth(Ribbon Width Scale=&FloatFromCurve(FloatCurve=DataInterface() {propertyValues {Curve {keys [
         {arriveTangent=-2.1863, interpMode=RCIM_Cubic, leaveTangent=-2.1863, tangentMode=RCTM_User, time=0, value=3}
         <at 0.4 0.55>
         {arriveTangent=-0.3076, interpMode=RCIM_Cubic, leaveTangent=-0.3076, tangentMode=RCTM_User, time=1, value=0}
        ]}}})) {module=ScaleRibbonWidth001}
  ]
  rendererClasses [&NiagaraRibbonRendererProperties]
  renderers [{rendererClass=&NiagaraRibbonRendererProperties}]
 }
}
summary {
 systemName=NS_Neon_Cat_Tech_Swing_Trail
 emitters [
  {bEnabled=false, emitterName=Trail_Swing, rendererClasses [&NiagaraRibbonRendererProperties]}
  {bEnabled=false, emitterName=Light, rendererClasses [&NiagaraLightRendererProperties]}
  {emitterName=AirTrail, rendererClasses [&NiagaraRibbonRendererProperties]}
  {bEnabled=false, emitterName=Empty, rendererClasses [&NiagaraSpriteRendererProperties]}
  {emitterName=BladeGlow, rendererClasses [&NiagaraMeshRendererProperties, &NiagaraSpriteRendererProperties]}
  {emitterName=FN_Attached_Ribbon_BETA, rendererClasses [&NiagaraRibbonRendererProperties]}
 ]
 userVariables [
  User.RibbonIDOffset {type=NiagaraInt32}
  User.SkeletalMesh {
   type=NiagaraDataInterfaceSkeletalMesh
   defaultValue {
    struct=NiagaraExt_VariableValue_DataInterface
    value {dataInterface=&self:NiagaraDataInterfaceSkeletalMesh_1}
   }
  }
 ]
}
